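# Sound cue for the WidePlane back bling's chest reaction in Athena.
# An equal-weight Random node picks one of five React wave variants, passes it through a short fixed delay, then a Modulator adds slight pitch and volume variation.
# It plays at reduced volume in the back bling sound class, with close-range pet attenuation.
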SoundCue /BRCosmetics/Athena/Sounds/BackBlings/WidePlane/IP_Backbling_WidePlane_ChestReact_Cue
root=Modulator_1
nodes [
 {id=Delay_0, children [Random_1], properties {delayMax=0.2, delayMin=0.2}}
 {id=Modulator_1, children [Delay_0], properties {pitchMax=1.02, pitchMin=0.98, volumeMax=1.02, volumeMin=0.98}}
 {id=Random_1, children [WavePlayer_5,WavePlayer_6,WavePlayer_7,WavePlayer_8,WavePlayer_9], properties {weights [1,1,1,1,1]}}
 {id=WavePlayer_5, properties {soundWaveAssetPtr=&/IP_Backpack_WidePlane_React_06}}
 {id=WavePlayer_6, properties {soundWaveAssetPtr=&/IP_Backpack_WidePlane_React_07}}
 {id=WavePlayer_7, properties {soundWaveAssetPtr=&/IP_Backpack_WidePlane_React_08}}
 {id=WavePlayer_8, properties {soundWaveAssetPtr=&/IP_Backpack_WidePlane_React_09}}
 {id=WavePlayer_9, properties {soundWaveAssetPtr=&/IP_Backpack_WidePlane_React_10}}
]
properties {
 attenuationSettings=&Fort_Pets_Close_Athena_Att
 maxDistance=1800.0001
 soundClassObject=/Game/Sounds/FortSoundClassesAndModes/Fort_Backblings
 volumeMultiplier=0.45
}
